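# Parent cloth material for the CountyFrog backpack, adding an animated flipbook face to the shared 2023 anime character shader.
# Offset, center-scaled UVs sample the face flipbook, masking Line Color into lit and shaded colors.
# PlayReactivity switches between the idle expression row and ID-selected reactive rows, with amount parameters scrubbing frames within each.
# Pose-change flags raise the Temporal Responsiveness output while the face changes.
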Material /BRCosmetics/Accessories/FORT_Backpacks/Backpack_M_MED_CountyFrog/Materials/M_Backpack_CountyFrog_Parent
graph {
 properties {
  bAutomaticallySetUsageInEditor=false
  bHasPixelAnimation=true
  bUseMaterialAttributes=true
  bUsedWithClothing=true
  bUsedWithMeshDeformer=true
  bUsedWithMorphTargets=true
  bUsedWithNiagaraMeshParticles=true
  bUsedWithSkeletalMesh=true
  bUsedWithStaticLighting=true
  pixelDepthOffsetMode=PDOM_Legacy
  shadingModel=MSM_Cloth
  usedShadingModels=Cloth
 }
 nodes [
  n1 = MaterialFunctionCall!(CustomUV) {materialFunction=/Game/Global/Materials/Character_Parent_2023/Anime_Parent_2023/MaterialFunctions/MFC_AnimeShader_2023_Pre}
  n2 = MaterialFunctionCall!(CustomLightVector, OverrideNormal_WS, STT=@n1.STT, Normal_Map=@n1.STT, DFL=@n1.STT, FX=@n1.STT) {materialFunction=/Game/Global/Materials/Character_Parent_2023/Anime_Parent_2023/MaterialFunctions/MFC_AnimeShader_2023_Calculations}
  n3 = VectorParameter() {group=CountyFrog, parameterName=Line Color, defaultValue=<rgba 0.0176 0.0116 0.0137 1>}
  MaterialFunctionCall!(
   Shading_Gradient=@n2.Shading_Gradient
   Threshold_Gradient=@n2.Shading_Gradient
   Specular_Gradient=@n2.Shading_Gradient
   Sheen_Mask=@n2.Shading_Gradient
   Thickness=@n2.Shading_Gradient
   DFL_Mask=@n2.Shading_Gradient
   FX=@n2.Shading_Gradient
   Metallic_Gradient=@n2.Shading_Gradient
   RimLight_Gradient=@n2.Shading_Gradient
   NormalBuffer=@n2.Shading_Gradient
   --------
   CINE_LightSources=@n2.Shading_Gradient
   CINE_LightColor=@n2.Shading_Gradient
   ---------
   Color_Lit=LinearInterpolate(A=@n1.Color_Lit, B=@n3.RGB, Alpha=NamedRerouteUsage()) {constB=0}
   Color_Shaded=LinearInterpolate(A=@n1.Color_Shaded, B=@n3.RGB, Alpha=NamedRerouteUsage()) {constB=0}
   SRM=@n1.SRM
   Emissive=@n1.SRM
   DFL_Color=@n1.SRM
   Specular_Color=@n1.SRM
   RimLight_Color=@n1.SRM
  ) {materialFunction=/Game/Global/Materials/Character_Parent_2023/Anime_Parent_2023/MaterialFunctions/MFC_AnimeShader_2023_Post}
  n4 = ScalarParameter() {group=CountyFrog, parameterName=UV_OffsetX, sliderMax=1, sliderMin=-1}
  n5 = ScalarParameter() {group=CountyFrog, parameterName=UV_OffsetY, sliderMax=1, sliderMin=-1}
  n6 = MaterialFunctionCall(
   UVs=Subtract(A=TextureCoordinate(), B=AppendVector(A=@n4, B=@n5))
   Texture Scale=ScalarParameter() {defaultValue=0.5, group=CountyFrog, parameterName=UVScale}
  ) {materialFunction=&ScaleUVsByCenter}
  n7 = ScalarParameter() {group=CountyFrog, parameterName=PlayReactivity, sliderMax=1}
  n8 = Min(A=ScalarParameter() {group=CountyFrog, parameterName=ExpressionAmount, sliderMax=1}, B=OneMinus(None=@n7))
  MaterialExpressionDivide_0 = Divide(A=Add(A=Min(A=@n8) {constB=0.999}) {constB=2}) {constB=3}
  n9 = ScalarParameter() {defaultValue=1, group=CountyFrog, parameterName=ReactiveExpressionAmount, sliderMax=1}
  n10 = ScalarParameter() {group=CountyFrog, parameterName=ReactiveExpressionID, sliderMax=1}
  MaterialExpressionDivide_2 = Divide(A=Add(A=Min(A=@n9) {constB=0.999}, B=Floor(None=@n10)) {constB=0}) {constB=3}
  n11 = Switch!(SwitchValue=@n7, Default=@MaterialExpressionDivide_0, Idle=@MaterialExpressionDivide_0, ActiveReactivity=@MaterialExpressionDivide_2) {inputs [
    {
     inputName=Idle
     input {
      expression=&self:MaterialExpressionDivide_0
      inputName=None
      mask=0
      maskA=0
      maskB=0
      maskG=0
      maskR=0
      outputIndex=0
     }
    }
    {
     inputName=ActiveReactivity
     input {
      expression=&self:MaterialExpressionDivide_2
      inputName=None
      mask=0
      maskA=0
      maskB=0
      maskG=0
      maskR=0
      outputIndex=0
     }
    }
   ]}
  n12 = MaterialFunctionCall("Animation  Phase (0-1)"=@n11, Number of Rows=Constant() {r=3}, Number of Columns=Constant() {r=6}, UVs=@n6.UVs) {materialFunction=&FlipBook}
  n13 = TextureObjectParameter() {parameterName=Face Flipbook, texture=/BRCosmetics/Accessories/FORT_Backpacks/Backpack_M_MED_CountyFrog/Textures/T_CountyFrog_Backpack_Face}
  n14 = TextureSample(UVs=@n12.UVs, Tex=RecordTextureStreamingInfo!(Tex=@n13, UVs=Divide(A=@n12.UVs) {constB=6}).Tex)
  NamedRerouteDeclaration(None=Multiply(A=@n6.0-1 mask, B=OneMinus(None=@n14.R))) {name=Face, nodeColor=<rgba 0 1 0.4141 1>}
  n15 = ScalarParameter() {group=CountyFrog Temporal Responsiveness, parameterName=HasIdlePoseChanged, sliderMax=1}
  n16 = ScalarParameter() {group=CountyFrog Temporal Responsiveness, parameterName=HasBlendPoseChanged, sliderMax=1}
  MaterialExpressionConstant_2 = Constant() {r=0.4}
  MaterialExpressionConstant_3 = Constant() {r=1}
  n17 = Switch!(
   SwitchValue=Max(A=Max(A=@n15, B=@n16), B=ScalarParameter() {group=CountyFrog Temporal Responsiveness, parameterName=HasReactivePoseChanged, sliderMax=1})
   Default=@MaterialExpressionConstant_2
   No=@MaterialExpressionConstant_2
   Yes=@MaterialExpressionConstant_3
  ) {
   constDefault=0.4
   inputs [
    {
     inputName=No
     input {
      expression=&self:MaterialExpressionConstant_2
      inputName=None
      mask=0
      maskA=0
      maskB=0
      maskG=0
      maskR=0
      outputIndex=0
     }
    }
    {
     inputName=Yes
     input {
      expression=&self:MaterialExpressionConstant_3
      inputName=None
      mask=0
      maskA=0
      maskB=0
      maskG=0
      maskR=0
      outputIndex=0
     }
    }
   ]
  }
  TemporalResponsivenessOutput!(None=@n17)
 ]
}
