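# Masked, cloth-shaded material for the CountyFrog backpack that renders thickened spline strands on skeletal meshes.
# SplineThicken drives world position offset and normal, with width from thickness parameters multiplied by vertex color and an optional pixel-depth fade toward a constant.
# Texture alpha masks the strands; black base color with dark LUT- and time-of-day-adjusted emissive feeds the shared character PostFX function.
# The per-vine thickness switch and second thicken function have no visible consumers.
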Material /BRCosmetics/Accessories/FORT_Backpacks/Backpack_M_MED_CountyFrog/Materials/M_Backpack_CountyFrog_SplineThicken
graph {
 expressions {
  MaterialExpressionAdd_7 {
   class=MaterialExpressionAdd
   inputs [
    A {from=MaterialExpressionPixelDepth_1, output=R}
    B {from=MaterialExpressionScalarParameter_15}
   ]
  }
  MaterialExpressionComposite_1 {class=MaterialExpressionComposite, inputs [], properties {subgraphName=Construct Base Inputs}}
  MaterialExpressionConstant3Vector_1 {class=MaterialExpressionConstant3Vector, properties {constant=<rgba 0.014 0.0107 0.0119 1>}}
  MaterialExpressionConstant3Vector_2 {class=MaterialExpressionConstant3Vector, properties {constant=<rgba 0 0.7011 1 1>}}
  MaterialExpressionConstant3Vector_3 {class=MaterialExpressionConstant3Vector, properties {constant=<rgba 0 0 1 0>}}
  MaterialExpressionConstant_0 {class=MaterialExpressionConstant}
  MaterialExpressionConstant_1 {class=MaterialExpressionConstant}
  MaterialExpressionConstant_3 {class=MaterialExpressionConstant}
  MaterialExpressionDivide_3 {
   class=MaterialExpressionDivide
   inputs [
    A {from=MaterialExpressionAdd_7}
    B {from=MaterialExpressionScalarParameter_16}
   ]
  }
  MaterialExpressionFresnel_0 {
   class=MaterialExpressionFresnel
   inputs [
    ExponentIn {from="", output=""}
    BaseReflectFractionIn {from="", output=""}
    Normal {from="", output=""}
   ]
   properties {baseReflectFraction=0, exponent=1}
  }
  MaterialExpressionLinearInterpolate_1 {
   class=MaterialExpressionLinearInterpolate
   inputs [
    A {from=MaterialExpressionScalarParameter_12}
    B {from=MaterialExpressionScalarParameter_11}
    Alpha {from=MaterialExpressionNamedRerouteUsage_8}
   ]
  }
  MaterialExpressionLinearInterpolate_3 {
   class=MaterialExpressionLinearInterpolate
   inputs [
    A {from=MaterialExpressionMultiply_6}
    B {from=MaterialExpressionConstant_1}
    Alpha {from=MaterialExpressionSaturate_5}
   ]
  }
  MaterialExpressionMaterialFunctionCall_0 {
   class=MaterialExpressionMaterialFunctionCall
   inputs [
    WidthBase {from=MaterialExpressionStaticSwitchParameter_4}
    WidthTip {from=MaterialExpressionStaticSwitchParameter_4}
    WorldPosition {from=""}
    UVs for Projection {from=""}
    UVs for Thickness {from=""}
    UVs for texturing {from=""}
    Expand U or V UV Channel {from=""}
    DeriveNormalZ {from=""}
    AngleCorrectedNormals {from=""}
    AdditionalNormal {from=""}
    FlattenPixelNormal {from=MaterialExpressionStaticBool_0}
    UVs For NormalShading {from=""}
   ]
   properties {materialFunction=&SplineThicken}
  }
  MaterialExpressionMaterialFunctionCall_1 {
   class=MaterialExpressionMaterialFunctionCall
   inputs [
    widthTip {from=MaterialExpressionStaticSwitchParameter_4}
    Flatten {from=MaterialExpressionReroute_0}
    widthBase {from=MaterialExpressionStaticSwitchParameter_4}
   ]
   properties {materialFunction=/Game/Characters/Player/Male/Medium/Bodies/M_MED_Neon/Materials/MF_splineThickenM}
  }
  MaterialExpressionMaterialFunctionCall_10 {
   class=MaterialExpressionMaterialFunctionCall
   inputs [
    Material {from=MaterialExpressionSetMaterialAttributes_0}
    M {from=MaterialExpressionReroute_4, output=M}
    Pre Skinned Local Position {from=MaterialExpressionReroute_5, output=Pre Skinned Local Position}
    Distance Blend {from=MaterialExpressionReroute_6, output=Distance Blend}
    Fresnel {from=MaterialExpressionReroute_7, output=Fresnel}
   ]
   properties {materialFunction=&MF_CharacterParent_2023_PostFX}
  }
  MaterialExpressionMaterialFunctionCall_2 {
   class=MaterialExpressionMaterialFunctionCall
   inputs [
    In {from=MaterialExpressionNamedRerouteDeclaration_0}
    LUT {from=""}
    LUT Multiplier {from=""}
   ]
   properties {materialFunction=/Game/Packages/Fortress_SharedMaterials/Anime_Master_Material/MF_ApplyLUT}
  }
  MaterialExpressionMaterialFunctionCall_25 {
   class=MaterialExpressionMaterialFunctionCall
   inputs [Color {from=MaterialExpressionMaterialFunctionCall_2, output=Result}]
   properties {materialFunction=/Game/Packages/Fortress_SharedMaterials/Anime_Master_Material/MF_TODColorAdjustment}
  }
  MaterialExpressionMaterialFunctionCall_3 {
   class=MaterialExpressionMaterialFunctionCall
   inputs [Material {from=""}]
   properties {materialFunction=/Game/Packages/Fortress_SharedMaterials/Base_Material_Functions/CharacterBrightnessandScale_Balance}
  }
  MaterialExpressionMultiply_4 {
   class=MaterialExpressionMultiply
   inputs [
    A {from=MaterialExpressionNamedRerouteUsage_4}
    B {from=""}
   ]
   properties {constB=5}
  }
  MaterialExpressionMultiply_5 {
   class=MaterialExpressionMultiply
   inputs [
    A {from=MaterialExpressionScalarParameter_0}
    B {from=MaterialExpressionLinearInterpolate_1}
   ]
  }
  MaterialExpressionMultiply_6 {
   class=MaterialExpressionMultiply
   inputs [
    A {from=MaterialExpressionMultiply_5}
    B {from=MaterialExpressionVertexColor_1}
   ]
  }
  MaterialExpressionNamedRerouteDeclaration_0 {
   class=MaterialExpressionNamedRerouteDeclaration
   inputs [None {from=MaterialExpressionConstant3Vector_1}]
   properties {name=baseColor, nodeColor=<rgba 0 1 0.2031 1>}
  }
  MaterialExpressionNamedRerouteDeclaration_1 {
   class=MaterialExpressionNamedRerouteDeclaration
   inputs [None {from=MaterialExpressionScalarParameter_3}]
   properties {name=metallic, nodeColor=<rgba 0.9922 0 1 1>}
  }
  MaterialExpressionNamedRerouteDeclaration_2 {
   class=MaterialExpressionNamedRerouteDeclaration
   inputs [None {from=MaterialExpressionScalarParameter_4}]
   properties {name=specular, nodeColor=<rgba 0.8984 0 1 1>}
  }
  MaterialExpressionNamedRerouteDeclaration_3 {
   class=MaterialExpressionNamedRerouteDeclaration
   inputs [None {from=MaterialExpressionScalarParameter_5}]
   properties {name=roughness, nodeColor=<rgba 0.5 0 1 1>}
  }
  MaterialExpressionNamedRerouteDeclaration_4 {
   class=MaterialExpressionNamedRerouteDeclaration
   inputs [None {from=MaterialExpressionMaterialFunctionCall_0, output=WorldPositionOffset}]
   properties {name=wpo, nodeColor=<rgba 0.125 0 1 1>}
  }
  MaterialExpressionNamedRerouteDeclaration_5 {
   class=MaterialExpressionNamedRerouteDeclaration
   inputs [None {from=MaterialExpressionMaterialFunctionCall_0, output=Normal}]
   properties {name=normal, nodeColor=<rgba 1 0.0703 0 1>}
  }
  MaterialExpressionNamedRerouteDeclaration_6 {
   class=MaterialExpressionNamedRerouteDeclaration
   inputs [None {from=MaterialExpressionVertexColor_0, output=R}]
   properties {name=VertColor_Red, nodeColor=<rgba 1 0.8438 0 1>}
  }
  MaterialExpressionNamedRerouteDeclaration_7 {
   class=MaterialExpressionNamedRerouteDeclaration
   inputs [None {from=MaterialExpressionVertexColor_0, output=G}]
   properties {name=VertColor_Green, nodeColor=<rgba 0 1 0.2031 1>}
  }
  MaterialExpressionNamedRerouteDeclaration_8 {
   class=MaterialExpressionNamedRerouteDeclaration
   inputs [None {from=MaterialExpressionTextureSampleParameter2D_1, output=A}]
   properties {name=opacity, nodeColor=<rgba 0 0.625 1 1>}
  }
  MaterialExpressionNamedRerouteUsage_4 {class=MaterialExpressionNamedRerouteUsage}
  MaterialExpressionNamedRerouteUsage_8 {class=MaterialExpressionNamedRerouteUsage}
  MaterialExpressionPinBase_0 {
   class=MaterialExpressionPinBase
   inputs [
    M {from=MaterialExpressionConstant_3, output=""}
    Pre Skinned Local Position {from=MaterialExpressionShaderStageSwitch_1, output=""}
    Distance Blend {from=MaterialExpressionMaterialFunctionCall_3, output=Distance Blend}
    Fresnel {from=MaterialExpressionFresnel_0, output=""}
   ]
   properties {reroutePins [
     M {expression=&self:MaterialExpressionReroute_4}
     Pre Skinned Local Position {expression=&self:MaterialExpressionReroute_5}
     Distance Blend {expression=&self:MaterialExpressionReroute_6}
     Fresnel {expression=&self:MaterialExpressionReroute_7}
    ]}
  }
  MaterialExpressionPinBase_1 {class=MaterialExpressionPinBase, inputs [], properties {}}
  MaterialExpressionPixelDepth_1 {class=MaterialExpressionPixelDepth, inputs [], properties {}}
  MaterialExpressionPreSkinnedPosition_1 {class=MaterialExpressionPreSkinnedPosition, inputs [], properties {}}
  MaterialExpressionReroute_0 {class=MaterialExpressionReroute, inputs [None {from=MaterialExpressionScalarParameter_2}]}
  MaterialExpressionReroute_4 {class=MaterialExpressionReroute, inputs [None {from=MaterialExpressionConstant_3}]}
  MaterialExpressionReroute_5 {class=MaterialExpressionReroute, inputs [None {from=MaterialExpressionShaderStageSwitch_1}]}
  MaterialExpressionReroute_6 {class=MaterialExpressionReroute, inputs [None {from=MaterialExpressionMaterialFunctionCall_3, output=Distance Blend}]}
  MaterialExpressionReroute_7 {class=MaterialExpressionReroute, inputs [None {from=MaterialExpressionFresnel_0}]}
  MaterialExpressionSaturate_5 {class=MaterialExpressionSaturate, inputs [None {from=MaterialExpressionDivide_3}]}
  MaterialExpressionScalarParameter_0 {class=MaterialExpressionScalarParameter, properties {defaultValue=2.5, parameterName=ThicknessOverallMult}}
  MaterialExpressionScalarParameter_1 {class=MaterialExpressionScalarParameter, properties {defaultValue=1, group=VineThickness, parameterName=VineThicknessMult_A}}
  MaterialExpressionScalarParameter_10 {class=MaterialExpressionScalarParameter, properties {defaultValue=1, group=VineThickness, parameterName=VineThicknessMult_F}}
  MaterialExpressionScalarParameter_11 {class=MaterialExpressionScalarParameter, properties {defaultValue=1, parameterName=Thickness_Min}}
  MaterialExpressionScalarParameter_12 {class=MaterialExpressionScalarParameter, properties {defaultValue=1, parameterName=Thickness_Max}}
  MaterialExpressionScalarParameter_15 {class=MaterialExpressionScalarParameter, properties {defaultValue=-5500, group=DistanceFade, parameterName=DistanceFadeOffset}}
  MaterialExpressionScalarParameter_16 {class=MaterialExpressionScalarParameter, properties {defaultValue=1500, group=DistanceFade, parameterName=DistanceFadeLength}}
  MaterialExpressionScalarParameter_2 {class=MaterialExpressionScalarParameter, properties {parameterName=Flatten}}
  MaterialExpressionScalarParameter_3 {class=MaterialExpressionScalarParameter, properties {defaultValue=0.5, parameterName=Metallic}}
  MaterialExpressionScalarParameter_4 {class=MaterialExpressionScalarParameter, properties {parameterName=Specular}}
  MaterialExpressionScalarParameter_5 {class=MaterialExpressionScalarParameter, properties {defaultValue=1, parameterName=Roughness}}
  MaterialExpressionScalarParameter_6 {class=MaterialExpressionScalarParameter, properties {defaultValue=1, group=VineThickness, parameterName=VineThicknessMult_B}}
  MaterialExpressionScalarParameter_7 {class=MaterialExpressionScalarParameter, properties {defaultValue=1, group=VineThickness, parameterName=VineThicknessMult_C}}
  MaterialExpressionScalarParameter_8 {class=MaterialExpressionScalarParameter, properties {defaultValue=1, group=VineThickness, parameterName=VineThicknessMult_D}}
  MaterialExpressionScalarParameter_9 {class=MaterialExpressionScalarParameter, properties {defaultValue=1, group=VineThickness, parameterName=VineThicknessMult_E}}
  MaterialExpressionSetMaterialAttributes_0 {
   class=MaterialExpressionSetMaterialAttributes
   inputs [
    MaterialAttributes {from="", output=""}
    Base Color {from=MaterialExpressionConstant_0, output=""}
    Metallic {from=MaterialExpressionNamedRerouteDeclaration_1, output=""}
    Specular {from=MaterialExpressionNamedRerouteDeclaration_2, output=""}
    Roughness {from=MaterialExpressionNamedRerouteDeclaration_3, output=""}
    Emissive Color {from=MaterialExpressionMaterialFunctionCall_25, output=Result}
    Opacity {from=MaterialExpressionConstant_0, output=""}
    Opacity Mask {from=MaterialExpressionNamedRerouteDeclaration_8, output=""}
    Normal {from=MaterialExpressionNamedRerouteDeclaration_5, output=""}
    World Position Offset {from=MaterialExpressionNamedRerouteDeclaration_4, output=""}
    Fuzz Color {from=MaterialExpressionConstant3Vector_2, output=""}
    Cloth {from=MaterialExpressionConstant_0, output=""}
   ]
   properties {attributeSetTypes [
     #1
     #2
     #3
     #4
     #5
     #6
     #7
     #8
     #9
     #10
     #11
    ]}
  }
  MaterialExpressionShaderStageSwitch_1 {
   class=MaterialExpressionShaderStageSwitch
   inputs [
    PixelShader {from=MaterialExpressionVertexInterpolator_1, output=PS}
    VertexShader {from=MaterialExpressionPreSkinnedPosition_1, output=""}
   ]
   properties {}
  }
  MaterialExpressionStaticBool_0 {class=MaterialExpressionStaticBool, properties {value=true}}
  MaterialExpressionStaticSwitchParameter_4 {
   class=MaterialExpressionStaticSwitchParameter
   inputs [
    True {from=MaterialExpressionLinearInterpolate_3, output=""}
    False {from=MaterialExpressionMultiply_6, output=""}
   ]
   properties {group=DistanceFade, parameterName=UseDistancefade}
  }
  MaterialExpressionSwitch_1 {
   class=MaterialExpressionSwitch
   inputs [
    SwitchValue {from=MaterialExpressionMultiply_4, output=""}
    Default {from="", output=""}
    None {from=MaterialExpressionScalarParameter_1, output=""}
    None {from=MaterialExpressionScalarParameter_6, output=""}
    None {from=MaterialExpressionScalarParameter_7, output=""}
    None {from=MaterialExpressionScalarParameter_8, output=""}
    None {from=MaterialExpressionScalarParameter_9, output=""}
    None {from=MaterialExpressionScalarParameter_10, output=""}
   ]
   properties {inputs [
     {
      inputName=None
      input {
       expression=&self:MaterialExpressionScalarParameter_1
       inputName=None
       mask=0
       maskA=0
       maskB=0
       maskG=0
       maskR=0
       outputIndex=0
      }
     }
     {
      inputName=None
      input {
       expression=&self:MaterialExpressionScalarParameter_6
       inputName=None
       mask=0
       maskA=0
       maskB=0
       maskG=0
       maskR=0
       outputIndex=0
      }
     }
     {
      inputName=None
      input {
       expression=&self:MaterialExpressionScalarParameter_7
       inputName=None
       mask=0
       maskA=0
       maskB=0
       maskG=0
       maskR=0
       outputIndex=0
      }
     }
     {
      inputName=None
      input {
       expression=&self:MaterialExpressionScalarParameter_8
       inputName=None
       mask=0
       maskA=0
       maskB=0
       maskG=0
       maskR=0
       outputIndex=0
      }
     }
     {
      inputName=None
      input {
       expression=&self:MaterialExpressionScalarParameter_9
       inputName=None
       mask=0
       maskA=0
       maskB=0
       maskG=0
       maskR=0
       outputIndex=0
      }
     }
     {
      inputName=None
      input {
       expression=&self:MaterialExpressionScalarParameter_10
       inputName=None
       mask=0
       maskA=0
       maskB=0
       maskG=0
       maskR=0
       outputIndex=0
      }
     }
    ]}
  }
  MaterialExpressionTextureSampleParameter2D_1 {
   class=MaterialExpressionTextureSampleParameter2D
   inputs [
    UVs {from="", output=""}
    Apply View MipBias {from="", output=""}
   ]
   properties {
    parameterName=DiffuseTexture
    samplerType=SAMPLERTYPE_Masks
    texture=/BRCosmetics/Accessories/FORT_Backpacks/Backpack_M_MED_CountyFrog/Textures/T_Backpack_CountyFrog_SplineThicken
   }
  }
  MaterialExpressionVertexColor_0 {class=MaterialExpressionVertexColor, inputs [], properties {}}
  MaterialExpressionVertexColor_1 {class=MaterialExpressionVertexColor, inputs [], properties {}}
  MaterialExpressionVertexInterpolator_1 {class=MaterialExpressionVertexInterpolator, inputs [VS {from=MaterialExpressionPreSkinnedPosition_1, output=""}], properties {}}
 }
 properties {
  bAutomaticallySetUsageInEditor=false
  bUseMaterialAttributes=true
  bUsedWithMeshDeformer=true
  bUsedWithSkeletalMesh=true
  blendMode=BLEND_Masked
  shadingModel=MSM_Cloth
  usedShadingModels=Cloth
 }
}
