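# Animation Blueprint for the CountyFrog backpack cosmetic, adding secondary motion to its dynamic paper and leg bones.
# The input pose passes through two additive paper bone transforms, then chained root-relative Anim Dynamics nodes, including two differently limited passes on dyn_paperRotate.
# Each update computes gravity and alpha variables from skydive floating, plus base offset variables from owner spine bend, with translation varying by gender.
# The AnimGraph shows no explicit reads of these variables.
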AnimBlueprint /BRCosmetics/Accessories/FORT_Backpacks/Backpack_M_MED_CountyFrog/Meshes/M_MED_CountyFrog_Pack_AnimBP
target_skeleton=&/M_MED_CountyFrog_Pack_Skeleton
graphs [
 AnimGraph {nodes [
   n1 = Input Pose()
   n2 = "Transform (Modify) Bone - Bone: dyn_paperRotate"(
    ComponentPose="Transform (Modify) Bone - Bone: dyn_paperbase"(ComponentPose=Local To Component(LocalPose=@n1), Alpha=1, Translation="0,0,0", Rotation="0,0,0", Scale="1,1,1") {rotationMode=BMM_Additive, scaleMode=BMM_Additive, translationMode=BMM_Additive, boneToModify {boneName=dyn_paperbase}}
    Alpha=1
    Translation="0,0,0"
    Rotation="0,0,0"
    Scale="1,1,1"
   ) {rotationMode=BMM_Additive, scaleMode=BMM_Additive, translationMode=BMM_Additive, boneToModify {boneName=dyn_paperRotate}}
   n3 = Anim Dynamics - Bone: dyn_paperRotate(ComponentPose=@n2, Alpha=1, ExternalForce="0,0,0") {
    angularDampingOverride=0.9999
    angularSpringConstant=5000
    bOverrideAngularDamping=true
    bUseGravityOverride=true
    numSolverIterationsPreUpdate=5
    simulationSpace=RootRelative
    boundBone {boneName=dyn_paperRotate}
    componentAppliedLinearAccClamp=<500 2000 2000>
    componentLinearAccScale=<0.1 0.35 0.35>
    componentLinearVelScale=<0.2 0.5 0.5>
    gravityOverride {z=-980}
    physicsBodyDefinitions [{
      boundBone {boneName=dyn_paperRotate}
      boxExtents=<15 37.5 15>
      constraintSetup {angularLimitsMax=<5 0 18>, angularLimitsMin=<-5 -15 -18>, angularTarget=<0 0 0>}
      localJointOffset=<3.5 0 0>
     }]
    relativeSpaceBone {boneName=pack}
   }
   n4 = Anim Dynamics - Bone: dyn_paperRotate!(ComponentPose=@n3, bAlphaBoolEnabled:bool=True, ExternalForce="0,0,0") -> (Pose) {
    alphaInputType=Bool
    angularDampingOverride=0.9999
    angularSpringConstant=5000
    bOverrideAngularDamping=true
    bUseGravityOverride=true
    numSolverIterationsPreUpdate=5
    simulationSpace=RootRelative
    boundBone {boneName=dyn_paperRotate}
    componentAppliedLinearAccClamp=<500 2000 2000>
    componentLinearAccScale=<0.1 0.35 0.35>
    componentLinearVelScale=<0.2 0.5 0.5>
    gravityOverride {z=-980}
    physicsBodyDefinitions [{
      boundBone {boneName=dyn_paperRotate}
      boxExtents=<15 37.5 15>
      constraintSetup {angularLimitsMax=<5 0 2>, angularLimitsMin=<-5 -15 2>, angularTarget=<0 0 0>}
      localJointOffset=<3.5 0 0>
     }]
    relativeSpaceBone {boneName=pack}
   }
   n5 = Anim Dynamics - Bone: dyn_paper_01!(ComponentPose=@n4, Alpha=1, GravityOverride:Vector="0,0,-980", ExternalForce="0,0,0") -> (Pose) {
    angularDampingOverride=0.9995
    angularSpringConstant=700
    bOverrideAngularDamping=true
    bUseGravityOverride=true
    numSolverIterationsPreUpdate=5
    simulationSpace=RootRelative
    boundBone {boneName=dyn_paper_01}
    componentAppliedLinearAccClamp=<500 2000 2000>
    componentLinearAccScale=<0.1 0.35 0.35>
    componentLinearVelScale=<0.1 0.5 0.5>
    gravityOverride {z=-980}
    physicsBodyDefinitions [{
      boundBone {boneName=dyn_paper_01}
      boxExtents=<15 25 15>
      constraintSetup {angularLimitsMax=<0 3 5>, angularLimitsMin=<0 -15 -5>, angularTarget=<0 0 0>}
      localJointOffset=<5.0111 0 0>
     }]
    relativeSpaceBone {boneName=dyn_paperRotate}
   }
   n6 = Anim Dynamics - Bone: dyn_paper_02!(ComponentPose=@n5, Alpha=1, GravityOverride:Vector="0,0,100", ExternalForce="0,0,0") -> (Pose) {
    angularDampingOverride=0.9995
    angularSpringConstant=700
    bOverrideAngularDamping=true
    bUseGravityOverride=true
    numSolverIterationsPreUpdate=5
    simulationSpace=RootRelative
    boundBone {boneName=dyn_paper_02}
    componentAppliedLinearAccClamp=<500 2000 2000>
    componentLinearAccScale=<0.1 0.35 0.35>
    componentLinearVelScale=<0.1 0.5 0.5>
    gravityOverride {z=100}
    physicsBodyDefinitions [{
      boundBone {boneName=dyn_paper_02}
      boxExtents=<15 25 15>
      constraintSetup {angularLimitsMax=<0 2.5 5>, angularLimitsMin=<0 -20 -5>, angularTarget=<0 0 0>}
      localJointOffset=<4.3342 0 0>
     }]
    relativeSpaceBone {boneName=dyn_paper_01}
   }
   n7 = Anim Dynamics - Bone: dyn_paper_03!(ComponentPose=@n6, Alpha=1, GravityOverride:Vector="0,0,100", ExternalForce="0,0,0") -> (Pose) {
    angularDampingOverride=0.9995
    angularSpringConstant=700
    bOverrideAngularDamping=true
    bUseGravityOverride=true
    numSolverIterationsPreUpdate=5
    simulationSpace=RootRelative
    boundBone {boneName=dyn_paper_03}
    componentAppliedLinearAccClamp=<500 2000 2000>
    componentLinearAccScale=<0.1 0.35 0.35>
    componentLinearVelScale=<0.1 0.5 0.5>
    gravityOverride {z=100}
    physicsBodyDefinitions [{
      boundBone {boneName=dyn_paper_03}
      boxExtents=<15 25 15>
      constraintSetup {angularLimitsMax=<0 2 0>, angularLimitsMin=<0 -25 0>, angularTarget=<0 0 0>}
      localJointOffset=<3.1491 0 0>
     }]
    relativeSpaceBone {boneName=root}
   }
   n8 = Anim Dynamics - Bone: dyn_L_upperLeg!(ComponentPose=@n7, Alpha=1, GravityOverride:Vector="0,0,100", ExternalForce="0,0,0") -> (Pose) {
    angularDampingOverride=0.99
    angularSpringConstant=350
    bOverrideAngularDamping=true
    bUseGravityOverride=true
    numSolverIterationsPreUpdate=5
    simulationSpace=RootRelative
    boundBone {boneName=dyn_L_upperLeg}
    componentAppliedLinearAccClamp=<1500 2000 2000>
    componentLinearAccScale=<0.3 0.35 0.35>
    componentLinearVelScale=<0.45 0.55 0.55>
    gravityOverride {z=100}
    physicsBodyDefinitions [{
      boundBone {boneName=dyn_L_upperLeg}
      boxExtents=<15 10 10>
      constraintSetup {angularLimitsMax=<0 10 10>, angularLimitsMin=<0 -10 -10>, angularTarget=<0 0 0>}
      localJointOffset=<5 0 0>
     }]
    relativeSpaceBone {boneName=root}
   }
   n9 = Anim Dynamics - Bone: dyn_L_lowerLeg!(ComponentPose=@n8, Alpha=1, GravityOverride:Vector="0,0,100", ExternalForce="0,0,0") -> (Pose) {
    angularDampingOverride=0.99
    angularSpringConstant=350
    bOverrideAngularDamping=true
    bUseGravityOverride=true
    numSolverIterationsPreUpdate=5
    simulationSpace=RootRelative
    boundBone {boneName=dyn_L_lowerLeg}
    componentAppliedLinearAccClamp=<1500 2000 2000>
    componentLinearAccScale=<0.4 0.45 0.45>
    componentLinearVelScale=<0.6 0.6 0.6>
    gravityOverride {z=100}
    physicsBodyDefinitions [{
      boundBone {boneName=dyn_L_lowerLeg}
      boxExtents=<20 10 10>
      constraintSetup {angularLimitsMax=<0 10 0>, angularLimitsMin=<0 -10 0>, angularTarget=<0 0 0>}
      localJointOffset=<5 0 0>
     }]
    relativeSpaceBone {boneName=root}
   }
   n10 = Anim Dynamics - Bone: dyn_R_upperLeg!(ComponentPose=@n9, Alpha=1, GravityOverride:Vector="0,0,100", ExternalForce="0,0,0") -> (Pose) {
    angularDampingOverride=0.99
    angularSpringConstant=350
    bOverrideAngularDamping=true
    bUseGravityOverride=true
    numSolverIterationsPreUpdate=5
    simulationSpace=RootRelative
    boundBone {boneName=dyn_R_upperLeg}
    componentAppliedLinearAccClamp=<1500 2000 2000>
    componentLinearAccScale=<0.3 0.35 0.35>
    componentLinearVelScale=<0.45 0.55 0.55>
    gravityOverride {z=100}
    physicsBodyDefinitions [{
      boundBone {boneName=dyn_R_upperLeg}
      boxExtents=<15 10 10>
      constraintSetup {angularLimitsMax=<0 10 10>, angularLimitsMin=<0 -10 -10>, angularTarget=<0 0 0>}
      localJointOffset=<5 0 0>
     }]
    relativeSpaceBone {boneName=root}
   }
   n11 = Anim Dynamics - Bone: dyn_R_lowerLeg!(ComponentPose=@n10, Alpha=1, GravityOverride:Vector="0,0,100", ExternalForce="0,0,0") -> (Pose) {
    angularDampingOverride=0.99
    angularSpringConstant=350
    bOverrideAngularDamping=true
    bUseGravityOverride=true
    numSolverIterationsPreUpdate=5
    simulationSpace=RootRelative
    boundBone {boneName=dyn_R_lowerLeg}
    componentAppliedLinearAccClamp=<1500 2000 2000>
    componentLinearAccScale=<0.4 0.45 0.45>
    componentLinearVelScale=<0.6 0.6 0.6>
    gravityOverride {z=100}
    physicsBodyDefinitions [{
      boundBone {boneName=dyn_R_lowerLeg}
      boxExtents=<20 10 10>
      constraintSetup {angularLimitsMax=<0 15 0>, angularLimitsMin=<0 0 0>, angularTarget=<0 0 0>}
      localJointOffset=<5 0 0>
     }]
    relativeSpaceBone {boneName=root}
   }
   Output Pose(Result=Component To Local(ComponentPose=@n11)) {layerGroup=None}
  ]}
 EventGraph {nodes [
   Event BlueprintUpdateAnimation()
   n1 = Get bIsSkydiveFloating()
   > Set!(
    execute
    Gravity Override Var_X:double=0
    Gravity Override Var_Y:double=0
    Gravity Override Var_Z:double=SelectFloat(A=200, B=-980, bPickA=@n1)
   ) -> (then, Output_Get:Vector)
   > Set DynBodyRotateGeneralAlpha(DynBodyRotateGeneralAlpha:double=SelectFloat(A=0, B=1, bPickA=@n1)) -> (Output_Get:double)
   > n2 = FindGender!(execute, Object:Object Reference=TryGetPawnOwner()) -> (True:Exec, False:Exec, OutGender:EFortCustomGender Enum=Invalid)
   n3 = "Equal (Enum)"(A:EFortCustomGender Enum=@n2, B:EFortCustomGender Enum=Female)
   n4 = SelectFloat(A=-0.4, B=-1.1, bPickA=@n3)
   n5 = SelectFloat(A=-0.1, B=-0.7, bPickA=@n3)
   n6 = GetOwnerSkeletalMeshComponent()
   n7 = BreakRotator(InRot=BreakTransform(InTransform=GetDeltaTransformFromRefPose(self=@n6, BoneName=spine_05, BaseName=spine_01)).Rotation)
   > Set ExtraBaseTrans(ExtraBaseTrans:Vector=MakeVector(Y=MapRangeClamped(Value=@n7.Yaw, InRangeA=5, InRangeB=30, OutRangeA=@n5, OutRangeB=@n4))) -> (Output_Get:Vector)
   > Set ExtraBaseRotate(ExtraBaseRotate:Rotator=MakeRotator(Roll=MapRangeClamped(Value=@n7.Yaw, InRangeA=5, InRangeB=30, OutRangeB=5))) -> (Output_Get:Rotator)
  ]}
]
variables [
 DynBodyRotateGeneralAlpha {type=number}
 ExtraBaseRotate {type=Rotator}
 ExtraBaseTrans {type=Vector}
 Gravity Override Var {type=Vector}
]
