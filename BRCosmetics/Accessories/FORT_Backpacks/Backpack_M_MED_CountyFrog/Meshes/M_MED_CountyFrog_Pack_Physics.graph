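# Physics asset for the CountyFrog backpack cosmetic, covering the dynamic bones of what the bone names suggest is a paper piece with left and right corners.
# Each of the three bones has a small capsule body, with the paper capsule offset and rotated from its bone.
# Both corners are constrained directly to the paper body with tight swing limits and a looser twist, so they can flex only slightly relative to it.
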PhysicsAsset /BRCosmetics/Accessories/FORT_Backpacks/Backpack_M_MED_CountyFrog/Meshes/M_MED_CountyFrog_Pack_Physics
bodies {
 dyn_L_corner [{shapeType=Capsule, length=0.505, radius=0.505, shapeName=dyn_L_corner_capsule}]
 dyn_R_corner [{shapeType=Capsule, length=0.505, radius=0.505, shapeName=dyn_R_corner_capsule}]
 dyn_paper_01 [{
   shapeType=Capsule
   length=0.505
   radius=0.505
   shapeName=dyn_paper_01_capsule
   center=<2.3697 0.0047 0.024>
   rotation=<rot 0 -0.4946 -89.9954>
  }]
}
constraints [
 {bone1Name=dyn_L_corner, bone2Name=dyn_paper_01, swing1LimitDegrees=1, swing2LimitDegrees=5, twistLimitDegrees=25}
 {bone1Name=dyn_R_corner, bone2Name=dyn_paper_01, swing1LimitDegrees=1, swing2LimitDegrees=5, twistLimitDegrees=25}
]
